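# Material function for grind rails that blends a glowing, black-based strip into incoming material attributes along the UV center line.
# Emissive color scrolls through hues over time or takes a solid color, shaped by Fresnel and brightness blends.
# A player-position sphere mask responds to sprint, jump-off and idle-tail switches in custom primitive data.
# Custom primitive data also remaps V into a unified UV range per primitive.
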MaterialFunction /GrindRail/RustedTrack/Materials/MF_GrindRail
graph {nodes [
  Constant() {r=1}
  Constant() {r=1}
  n1 = FunctionInput() {inputName=MA, inputType=FunctionInput_MaterialAttributes}
  n2 = GetMaterialAttributes(None=@n1)
  n3 = NamedRerouteUsage()
  n4 = Time()
  n5 = ScalarParameter() {defaultValue=0.05, group=GrindRail, parameterName=EmissiveSpeed}
  n6 = Frac(None=Add(A=ComponentMask(None=@n3) {g=true}, B=Multiply(A=@n4, B=@n5)))
  n7 = LinearInterpolate(
   A=MaterialFunctionCall!(
    Base=Constant3Vector() {constant=<rgba 1 0 0 0>}
    Hue=LinearInterpolate(Alpha=@n6) {constA=-360, constB=360}
    Saturation
    Value
    Effect Amount
   ) {materialFunction=/BRCosmetics/Weapons/FORT_Melee/Pickaxe_QuailWink/MF_DM_Effect_Base_HSV}.Result
   B=VectorParameter() {group=GrindRail, parameterName=Rail Solid Color, defaultValue=<rgba 0.6302 0.6302 0.6302 1>}.RGB
   Alpha=ScalarParameter() {desc=Used to set glow solid colors, group=GrindRail, parameterName=EmissiveSolidColor}
  ) {constB=0.5}
  n8 = LinearInterpolate(A=@n7, B=Constant3Vector() {constant=<rgba 1 1 1 0>}, Alpha=NamedRerouteUsage())
  n9 = NamedRerouteUsage()
  n10 = Fresnel(ExponentIn=ScalarParameter() {defaultValue=0.5, group=GrindRail, parameterName=EmissiveFresnelExp}) {baseReflectFraction=0}
  n11 = LinearInterpolate(Alpha=NamedRerouteDeclaration(None=OneMinus(None=@n10)) {name=Fresnel, nodeColor=<rgba 0.0547 1 0 1>}) {constA=0.05}
  n12 = NamedRerouteUsage()
  n13 = Multiply(
   A=Multiply(A=LinearInterpolate(A=@n7, B=@n8, Alpha=@n9), B=LinearInterpolate(A=@n11, Alpha=@n12))
   B=LinearInterpolate(
    A=ScalarParameter() {defaultValue=5, group=GrindRail, parameterName=EmissiveBrightness}
    B=ScalarParameter() {defaultValue=32, group=GrindRail, parameterName=PlayerSphereBrightness}
    Alpha=NamedRerouteUsage()
   )
  )
  n14 = LinearInterpolate(
   A=@n13
   B=MaterialFunctionCall!(Input=@n13, amount=Constant3Vector() {constant=<rgba 0 1 2 0>}) {materialFunction=/InterchangeAssets/Functions/MX_HSVAdjust}.Result
   Alpha=NamedRerouteUsage()
  )
  n15 = Constant()
  n16 = ComponentMask(None=TextureCoordinate()) {r=true}
  n17 = Constant() {r=0.5}
  FunctionOutput(None=BlendMaterialAttributes(
    A=@n1
    B=SetMaterialAttributes!(MaterialAttributes=@n2.MaterialAttributes, EmissiveColor=@n14, BaseColor=@n15) {attributeSetTypes [#1, #2]}
    Alpha=SphereMask(A=@n16, B=@n17) {attenuationRadius=0.1}
   ))
  n18 = TextureCoordinate()
  MaterialExpressionRemap_0 = MaterialXRemap!(
   None=OneMinus(None=ComponentMask(None=@n18) {g=true})
   InputLow
   InputHigh
   TargetLow=ScalarParameter() {bUseCustomPrimitiveData=true, parameterName=Unified UV Start}
   TargetHigh=ScalarParameter() {bUseCustomPrimitiveData=true, parameterName=Unified UV End, primitiveDataIndex=1}
  )
  NamedRerouteDeclaration(None=AppendVector(A=ComponentMask(None=@n18) {r=true}, B=@MaterialExpressionRemap_0)) {name=Unified UV, nodeColor=<rgba 0.4531 1 0 1>}
  n19 = WorldPosition()
  n20 = CollectionParameter() {collection=&FortniteMaterialParameters, parameterName=PlayerPosition}
  n21 = ComponentMask(None=@n20) {b=true, g=true, r=true}
  n22 = LinearInterpolate(
   A=LinearInterpolate(
    A=ScalarParameter() {defaultValue=400, group=EmissiveColor, parameterName=RadiusDefaultEffect}
    B=ScalarParameter() {defaultValue=900, group=EmissiveColor, parameterName=RadiusSprintingEffectRadius}
    Alpha=ScalarParameter() {bUseCustomPrimitiveData=true, group=Gameplay, parameterName=sprinting state, primitiveDataIndex=8}
   )
   Alpha=ScalarParameter() {bUseCustomPrimitiveData=true, defaultValue=1, parameterName=IdleTail light switch, primitiveDataIndex=6}
  ) {constB=0}
  n23 = ScalarParameter() {group=EmissiveColor, parameterName=HardnessRadius}
  n24 = Multiply(A=Power(Base=SphereMask(A=@n19.XYZ, B=@n21, Radius=@n22, Hardness=@n23)), B=Constant() {r=1.02})
  n25 = ScalarParameter() {bUseCustomPrimitiveData=true, group=Gameplay, parameterName=Jumpoff, primitiveDataIndex=7}
  n26 = ScalarParameter() {bUseCustomPrimitiveData=true, group=Gameplay, parameterName=sprinting state, primitiveDataIndex=8}
  NamedRerouteDeclaration(None=Multiply(A=LinearInterpolate(A=@n24, Alpha=@n25) {constB=0}, B=LinearInterpolate(Alpha=@n26) {constA=0.5})) {name=Player SphereMask, nodeColor=<rgba 0.5 0 1 1>}
 ]}
