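# Prototype master material M_Track_01_Proto_1 for the GrindRail rusted track assets, using the Default Lit shading model.
# It holds a single vector parameter with a dark, slightly blue-grey default color, and the graph shows no connections to material outputs.
# Automatic usage-flag setting in the editor is disabled.
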Material /GrindRail/RustedTrack/Materials/M_Track_01_Proto_1
graph {
 properties {bAutomaticallySetUsageInEditor=false, usedShadingModels=Default Lit}
 nodes [VectorParameter() {defaultValue=<rgba 0.0312 0.0323 0.0469 0>}]
}
